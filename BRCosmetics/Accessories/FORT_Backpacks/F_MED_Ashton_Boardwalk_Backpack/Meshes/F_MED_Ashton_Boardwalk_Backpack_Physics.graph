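# Physics asset for the Ashton Boardwalk backpack cosmetic, defining collision and joints for two dynamic baton bones on the pack.
# A capsule body on pack_mid forms the base, and each baton has its own long, thin capsule.
# Each baton is constrained directly to pack_mid with tight, equal swing and twist limit angles, with no constraint between the two batons.
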PhysicsAsset /BRCosmetics/Accessories/FORT_Backpacks/F_MED_Ashton_Boardwalk_Backpack/Meshes/F_MED_Ashton_Boardwalk_Backpack_Physics
bodies {
 Dyn_baton01 [{shapeType=Capsule, length=30.1681, radius=1.4658, center=<-4.9969 -0.2704 -0.2624>, rotation=<rot 0 92.0932 -86.5302>}]
 Dyn_baton02 [{shapeType=Capsule, length=30.1671, radius=1.3124, center=<-5.4451 -0.0064 0.1177>, rotation=<rot 0 -90.7535 -88.8244>}]
 pack_mid [{shapeType=Capsule, length=11.25, radius=8}]
}
constraints [
 {bone1Name=Dyn_baton01, bone2Name=pack_mid, swing1LimitDegrees=5, swing2LimitDegrees=5, twistLimitDegrees=5}
 {bone1Name=Dyn_baton02, bone2Name=pack_mid, swing1LimitDegrees=5, swing2LimitDegrees=5, twistLimitDegrees=5}
]
