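# Animation Blueprint for the Ashton Boardwalk backpack cosmetic, adding rigid body secondary motion to the incoming pose.
# It caches full-strength dynamics and a toned-down version with halved component-motion scales, half-blended back toward the input from Dyn_baton_main.
# The toned-down pose plays when airborne or targeting, full dynamics otherwise.
# Each update sets IsInAir from jumping/falling flags, then overwrites it using a vertical velocity threshold.
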AnimBlueprint /BRCosmetics/Accessories/FORT_Backpacks/F_MED_Ashton_Boardwalk_Backpack/Meshes/F_MED_Ashton_Boardwalk_Backpack_AnimBP
target_skeleton=&/F_MED_Ashton_Boardwalk_Backpack_Skeleton
graphs [
 AnimGraph {nodes [
   n1 = Use cached pose 'ToneDown50%'()
   n2 = Use cached pose 'FullDynamics'()
   n3 = Get bIsTargeting()
   Output Pose(Result=Blend Poses by bool(
     BlendPose_0=Use cached pose 'ToneDown50%'()
     BlendPose_1=Blend Poses by bool(BlendPose_0=@n1, BlendPose_1=@n2, bActiveValue=@n3) {blendType=Linear}
     bActiveValue=Get IsInAir() -> (IsInAir:bool)
    ) {blendType=Linear}) {layerGroup=None}
   AnimGraphNode_SubInput_0 = Input Pose()
   Save cached pose 'SubGraph_Input'(Pose=@AnimGraphNode_SubInput_0) {cacheName=SubGraph_Input}
   n4 = RigidBody(ComponentPose=Local To Component(LocalPose=Use cached pose 'SubGraph_Input'())) {
    bOverrideWorldGravity=true
    overridePhysicsAsset=&/F_MED_Ashton_Boardwalk_Backpack_Physics
    alphaBoolBlend {blendOption=Linear}
    baseBoneRef {boneName=root}
    componentLinearAccScale=<0.5 0.5 0.5>
    componentLinearVelScale=<0.5 0.5 0.5>
    overrideWorldGravity {z=-980}
   }
   Save cached pose 'ToneDown50%'(Pose=Layered blend per bone(BasePose=Component To Local(ComponentPose=@n4), BlendPoses_0=Use cached pose 'SubGraph_Input'(), BlendWeights_0=0.5) {layerSetup [{branchFilters [{blendDepth=1, boneName=Dyn_baton_main}]}]}) {cacheName=ToneDown50%}
   n5 = RigidBody(ComponentPose=Local To Component(LocalPose=Use cached pose 'SubGraph_Input'())) {
    bOverrideWorldGravity=true
    overridePhysicsAsset=&F_MED_Ashton_Boardwalk_Backpack_Physics
    alphaBoolBlend {blendOption=Linear}
    baseBoneRef {boneName=root}
    componentLinearAccScale=<1 1 1>
    componentLinearVelScale=<1 1 1>
    overrideWorldGravity {z=-980}
   }
   Save cached pose 'FullDynamics'(Pose=Component To Local(ComponentPose=@n5)) {cacheName=FullDynamics}
   Get bIsJumping()
  ]}
 EventGraph {nodes [
   Event BlueprintUpdateAnimation()
   > n1 = Sequence()
   > Set bUpdateOwnerAnimInputProperty(bUpdateOwnerAnimInputProperty:bool=true) -> (Output_Get:bool)
   > n2 = Branch(Condition=OR Boolean(A=Get bIsJumping(), B=Get bIsFalling()))
   > Set IsInAir(IsInAir:bool=true) -> (Output_Get:bool)
   Set IsInAir(execute=@n2.else, IsInAir=false)
   n3 = Get PawnVelocityZ()
   n4 = Branch(
    execute=@n1.then_1
    Condition=OR Boolean(
     A="float > float"(A=@n3, B=0.1) {is=title K2Node_CallFunction}
     B="float < float"!(A:double=@n3, B:double=-0.1) -> (ReturnValue:bool) {is=title K2Node_CallFunction}
    )
   )
   > Set IsInAir(IsInAir=true)
   Set IsInAir(execute=@n4.else, IsInAir=false)
   TryGetPawnOwner()
  ]}
]
variables [IsInAir {type=boolean}]
